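# Translucent UI material for the Fortnitemares S37 item shop tile background, with grunge stripes across top and bottom.
# Two channels of one grunge texture, scaled, offset and clipped by edge masks, form the stripes, joined across the tile's vertical halves.
# The stripe value indexes a three-color lerp fed by shared shop background parameters, blended evenly with a linear gradient in those colors.
# Shared functions handle UV fitting and rounded corners.
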Material /ItemShop/UI/Art/FNM/S37/M_UI_FNMaresS37_ShopTile
graph {
 properties {
  bAutomaticallySetUsageInEditor=false
  bUsedWithStaticMesh=false
  blendMode=BLEND_Translucent
  materialDomain=MD_UI
  usedShadingModels=Default Lit
 }
 nodes [
  Constant4Vector() {constant=<rgba 0.4348 0.0154 0.3454 1>}
  Constant4Vector() {constant=<rgba 0.9387 0.147 0.0116 1>}
  Constant4Vector() {constant=<rgba 0.6654 0.0356 0.0887 1>}
  n1 = MaterialFunctionCall() {materialFunction=/FortUILibrary/Systems/ItemShopTile/Materials/MaterialFunctions/MF_UI_ItemShop_BackgroundParams}
  n2 = NamedRerouteDeclaration(None=@n1.Color 3) {name=color 3, nodeColor=<rgba 1 0 0.3984 1>}
  n3 = NamedRerouteDeclaration(None=@n1.Color 2) {name=color 2, nodeColor=<rgba 0.0781 1 0 1>}
  n4 = NamedRerouteDeclaration(None=@n1.Color 1) {name=color 1, nodeColor=<rgba 0.2422 0 1 1>}
  n5 = MaterialFunctionCall(UVs=NamedRerouteUsage(), Texture Scale=ScalarParameter() {defaultValue=1, group=TopStripe, parameterName=TopScale}) {materialFunction=&ScaleUVsByCenter}
  n6 = AppendVector(A=Constant(), B=ScalarParameter() {defaultValue=0.1107, group=TopStripe, parameterName=YPosition})
  n7 = Saturate(None=TextureSample(UVs=Add(A=@n5.UVs, B=@n6)) {samplerType=SAMPLERTYPE_Normal, texture=/ItemShop/UI/Art/FNM/S37/Textures/T_UI_GrungeStrip}.R)
  n8 = NamedRerouteUsage()
  n9 = Saturate(None=LinearInterpolate(Alpha=Min(A=@n7, B=@n8)) {constA=1, constB=0.6667})
  n10 = Add(
   A=MaterialFunctionCall(UVs=NamedRerouteUsage(), Texture Scale=ScalarParameter() {defaultValue=1, group=BottomStripe, parameterName=BottomScale}) {materialFunction=&ScaleUVsByCenter}.UVs
   B=AppendVector(A=Constant(), B=ScalarParameter() {defaultValue=-0.1033, group=BottomStripe, parameterName=BottomYPosition})
  )
  n11 = Min(
   A=Saturate(None=TextureSample(UVs=@n10) {samplerType=SAMPLERTYPE_Normal, texture=&T_UI_GrungeStrip}.G)
   B=NamedRerouteUsage()
  )
  n12 = Saturate(None=LinearInterpolate(Alpha=@n11) {constB=0.3333})
  n13 = ComponentMask(None=MaterialFunctionCall() {materialFunction=&GetUserInterfaceUV}.Normalized UV) {g=true}
  n14 = Saturate(None=Max(A=@n11, B=Round(None=@n13)))
  n15 = NamedRerouteUsage()
  n16 = Constant()
  n17 = Constant() {r=1}
  n18 = Constant() {r=0.25}
  LinearInterpolate(
   A=MaterialFunctionCall(1=@n2, 2=@n3, 3=@n4, A=LinearInterpolate(A=@n9, B=@n12, Alpha=@n14)) {materialFunction=&Lerp_Multiple_Float3}.Lerp 3 Inputs
   B=MaterialFunctionCall(
    1=@n2
    2=@n3
    3=@n4
    A=MaterialFunctionCall(UVs=@n15, GradientStart=@n16, GradientEnd=@n17, RotationAngle=@n18) {materialFunction=&MF_UI_LinearGradient}.Result
   ) {materialFunction=&Lerp_Multiple_Float3}.Lerp 3 Inputs
  )
  MaterialFunctionCall!(Input Alpha=Constant() {r=1}, bOverrideCornerRadius, Corner Radius Override) {materialFunction=/FortUILibrary/Systems/ItemShopTile/Materials/MaterialFunctions/MF_UI_ItemShop_RoundedCorners}
  MaterialFunctionCall!(ScreenUV, TextureSize, ScalePivotPoint) {materialFunction=/FortUILibrary/MaterialFunctions/Basics/MF_UI_ScreenSpaceUVScaling}
  NamedRerouteDeclaration(None=MaterialFunctionCall(UV=MaterialFunctionCall() {materialFunction=&GetUserInterfaceUV}.9-Slice UV) {materialFunction=&MF_UI_UVScaling}.ScaleY) {name=UVFit, nodeColor=<rgba 0 1 0.1797 1>}
  n19 = ConstantBiasScale(None=NamedRerouteUsage()) {bias=-0.5, scale=2}
  n20 = Subtract(A=Multiply(A=ComponentMask(None=@n19) {g=true}) {constB=-1}, B=Constant())
  NamedRerouteDeclaration(None=OneMinus(None=SmoothStep(Value=@n20) {constMax=0.6, constMin=0.6})) {name=Bottom_EdgeMask, nodeColor=<rgba 1 0 0.0234 1>}
  n21 = ConstantBiasScale(None=NamedRerouteUsage()) {bias=-0.5, scale=2}
  n22 = Subtract(A=ComponentMask(None=Abs(None=@n21)) {g=true}, B=Constant())
  NamedRerouteDeclaration(None=OneMinus(None=SmoothStep(Value=@n22) {constMax=0.95, constMin=0.8})) {name=Top_EdgeMask, nodeColor=<rgba 0 0.9062 1 1>}
 ]
}
